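# Battery gauge material for the Nevada light-beam vehicle, showing charge as a glowing fill over a dark background.
# Flooring UV V plus BatteryCharge creates a hard fill mask without textures.
# The fill blends from ColorCharged to ColorNoCharge along V, so its color follows position on the gauge.
# Red vertex color darkens fill and background, and the masked fill times EmissiveIntensity supplies the glow.
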Material /Hera/Props/LightBeam/LightBeam_Nevada_Vehicle_A/Materials/M_Nevada_Battery_Gauge
graph {
 properties {
  bUsedWithMeshDeformer=true
  bUsedWithNanite=true
  bUsedWithSkeletalMesh=true
  pixelDepthOffsetMode=PDOM_Legacy
  previewMesh=/Nevada/FX/Meshes/sm_nevada_battery_gauge
  twoSided=true
  usedShadingModels=Default Lit
 }
 nodes [
  Constant() {r=0.35}
  n1 = VectorParameter() {parameterName=BG_Color, defaultValue=<rgba 0.0447 0.0471 0.0747 0>}
  n2 = VertexColor()
  n3 = ComponentMask(None=TextureCoordinate()) {g=true}
  n4 = LinearInterpolate(
   A=VectorParameter() {parameterName=ColorCharged, defaultValue=<rgba 0 0.6517 1 0>}.RGB
   B=VectorParameter() {parameterName=ColorNoCharge, defaultValue=<rgba 1 0 0 0>}.RGB
   Alpha=@n3
  )
  n5 = LinearInterpolate(A=@n4, B=Multiply(A=@n4) {constB=0.025}, Alpha=@n2.R)
  n6 = Floor(None=Add(A=@n3, B=ScalarParameter() {defaultValue=1, parameterName=BatteryCharge}) {constB=0.8})
  LinearInterpolate(A=LinearInterpolate(A=@n1.RGB, B=Multiply(A=@n1.RGB) {constB=0.65}, Alpha=@n2.R), B=@n5, Alpha=@n6)
  Multiply(A=LinearInterpolate(B=@n5, Alpha=@n6), B=ScalarParameter() {defaultValue=1, parameterName=EmissiveIntensity})
 ]
}
