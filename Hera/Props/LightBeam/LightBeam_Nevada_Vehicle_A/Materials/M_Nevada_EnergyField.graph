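# Translucent, two-sided unlit material for the energy field of the Nevada vehicle light beam prop, with a hue-shift control.
# Panning noise distorts plasma-cloud and fluid-flame layers, combined with a magenta radial rim from an inverted sphere mask.
# Growing Mesh blends to a curve-atlas gradient with vertex-green opacity, removing refraction and normal distortion.
# Quality and shading-path switches drop depth fade and distortion on Low, Medium, Forward and Mobile.
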Material /Hera/Props/LightBeam/LightBeam_Nevada_Vehicle_A/Materials/M_Nevada_EnergyField
graph {
 properties {
  bUsedWithMeshDeformer=true
  bUsedWithNanite=true
  bUsedWithNiagaraMeshParticles=true
  bUsedWithSkeletalMesh=true
  blendMode=BLEND_Translucent
  pixelDepthOffsetMode=PDOM_Legacy
  refractionMethod=RM_IndexOfRefraction
  shadingModel=MSM_Unlit
  twoSided=true
  usedShadingModels=Unlit
 }
 nodes [
  n1 = Constant3Vector() {constant=<rgba 0.1833 0 1 1>}
  n2 = Constant3Vector() {constant=<rgba 0.0648 0.0156 1 1>}
  n3 = TextureCoordinate() {uTiling=0.5, vTiling=0.5}
  n4 = Panner() {speedX=0.0286, speedY=0.0476}
  n5 = TextureSample(UVs=Panner(Coordinate=TextureCoordinate()) {speedX=-0.0762, speedY=-0.0095}) {texture=&T_Inky_Smoke_Tile}
  n6 = TextureSample(UVs=Panner(Coordinate=TextureCoordinate() {uTiling=0.5, vTiling=0.5}) {speedX=0.0095, speedY=0.0095}) {texture=/Game/Global/Textures/Default/Noise/T_Tiling_Noise_01}
  n7 = Add(
   A=Multiply(A=TextureSample(UVs=@n4) {samplerType=SAMPLERTYPE_Grayscale, texture=&LowResBlurredNoise}.R) {constB=0.3}
   B=Add(A=Multiply(A=@n5.R) {constB=-0.5}, B=Multiply(A=@n6.R) {constB=-0.2})
  )
  n8 = OneMinus(None=@n7)
  n9 = TextureSample(UVs=Add(A=Panner(Coordinate=@n3) {speedX=-0.0095, speedY=0.0071}, B=Multiply(A=@n8) {constB=0.5})) {texture=/Game/Weapons/WeaponSkins/Wraps/Textures/PlasmaClouds/T_Wrap_PlasmaClouds__M}
  n10 = Constant3Vector() {constant=<rgba 0.0423 0.026 1 1>}
  n11 = Constant3Vector() {constant=<rgba 0.9843 0.0573 1 1>}
  n12 = TextureCoordinate() {uTiling=0.7, vTiling=0.7}
  n13 = TextureSample(UVs=Add(A=Panner(Coordinate=@n12) {speedY=0.0205}, B=Multiply(A=@n7) {constB=0.5})) {samplerType=SAMPLERTYPE_LinearGrayscale, texture=/Game/Weapons/FORT_Melee/Pickaxe_Cube_Ninja_Male/FX/T_FluidFlameTile}
  n14 = Add(
   A=Multiply(A=LinearInterpolate(A=@n1, B=@n2, Alpha=@n9.G), B=Multiply(A=@n9.G) {constB=1.5})
   B=Multiply(A=LinearInterpolate(A=@n10, B=@n11, Alpha=@n13.R), B=Multiply(A=@n13.R) {constB=2})
  )
  n15 = Constant() {r=0.5}
  n16 = OneMinus(None=SphereMask(A=TextureCoordinate(), B=@n15, Radius=@n15, Hardness=Constant()))
  n17 = Add(A=Multiply(A=Power(Base=@n16) {constExponent=25}) {constB=10}, B=Multiply(A=@n16) {constB=0.5})
  n18 = Multiply(A=@n17, B=Constant3Vector() {constant=<rgba 0.9207 0.048 1 1>})
  n19 = CurveAtlasRowParameter(CurveTime=ComponentMask(None=TextureCoordinate()) {g=true}) {
   atlas=/Game/Athena/Items/AwfulDuck/FX/CA_AwfulDuck
   curve=/Game/Athena/Items/AwfulDuck/FX/Curve_AwfulDuck_04
   defaultValue=3
   parameterName=Curve
  }
  n20 = ScalarParameter() {parameterName=Growing Mesh}
  MaterialFunctionCall(Hue Shift Percentage=ScalarParameter() {parameterName=Color Hue Shift}, Texture=LinearInterpolate(A=Add(A=@n14, B=@n18), B=@n19, Alpha=@n20)) {materialFunction=&HueShift}
  n21 = LinearInterpolate(A=Add(A=Add(A=@n9.G, B=@n13.R), B=@n17), B=Multiply(A=VertexColor().G, B=@n19.A), Alpha=@n20)
  n22 = DepthFade(Opacity=@n21) {fadeDistanceDefault=10}
  n23 = QualitySwitch(Default=@n22, Low=@n21, High=@n22, Medium=@n21, Epic=@n22)
  n24 = ShadingPathSwitch(Default=@n23, Deferred=@n23, Forward=@n21, Mobile=@n21)
  Multiply(A=@n24, B=ScalarParameter() {defaultValue=1, parameterName=EnergyFieldOpacity})
  n25 = Multiply(A=@n24) {constB=2}
  n26 = ScalarParameter() {defaultValue=1, parameterName=EnergyFieldOpacity}
  n27 = Constant() {r=1}
  n28 = LinearInterpolate(A=LinearInterpolate(Alpha=Multiply(A=@n25, B=@n26)) {constA=1, constB=1.05}, B=@n27, Alpha=@n20)
  n29 = QualitySwitch(Default=@n28, Low=@n27, High=@n28, Medium=@n27, Epic=@n28)
  ShadingPathSwitch(Default=@n29, Deferred=@n29, Forward=@n27, Mobile=@n27)
  n30 = Constant3Vector() {constant=<rgba 0 0 1 0>}
  n31 = LinearInterpolate(
   A=TextureSample(UVs=Panner() {speedX=-0.0095, speedY=0.0071}) {samplerType=SAMPLERTYPE_Normal, texture=&T_Weap_Distort_N_02}.RGB
   B=@n30
   Alpha=@n20
  )
  n32 = QualitySwitch(Default=@n31, Low=@n30, High=@n31, Medium=@n30, Epic=@n31)
  ShadingPathSwitch(Default=@n32, Deferred=@n32, Forward=@n30, Mobile=@n30)
 ]
}
